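# Translucent unlit Niagara material for the lava splash effect of the Renegade Raider Spark pickaxe.
# A noise texture pans in polar UVs, sampled with explicit derivatives, and feeds MF_Erosion for opacity and a multi-color lava ramp.
# Erosion amount comes from radial center and edge masks, particle alpha, dynamic parameters and a linear gradient, with depth fade skipped at low and medium quality.
# A previous-frame switch adds a radial offset for motion blur.
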Material /Game/Weapons/FORT_Melee/Pickaxe_RenegadeRaider_Spark/FX/Materials/M_Pickaxe_RenegadeRaider_Spark_LavaSplash
graph {
 properties {
  bUsedWithNiagaraMeshParticles=true
  bUsedWithNiagaraRibbons=true
  bUsedWithNiagaraSprites=true
  blendMode=BLEND_Translucent
  pixelDepthOffsetMode=PDOM_Legacy
  shadingModel=MSM_Unlit
  usedShadingModels=Unlit
 }
 nodes [
  n1 = ParticleColor()
  n2 = NamedRerouteUsage()
  n3 = Power(Base=Multiply(A=NamedRerouteUsage(), B=@n2), Exp=ScalarParameter() {defaultValue=1, parameterName=Color Lerp Exponent})
  n4 = Multiply(A=@n1.A, B=@n3)
  n5 = VectorParameter() {parameterName=Color 4, defaultValue=<rgba 0.04 0.0034 0.0028 3>}
  n6 = NamedRerouteUsage()
  n7 = MaterialFunctionCall!(Gradient=@n2, Erosion Amount=@n6, Softness=ScalarParameter() {defaultValue=0.333, parameterName=Erosion Color Softness}) {materialFunction=/Game/Effects/Fort_Effects/MaterialFunctions/MF_Erosion}
  n8 = LinearInterpolate(
   A=MaterialFunctionCall(
    1=VectorParameter() {parameterName=Color 1, defaultValue=<rgba 3 0.45 0 0>}.RGB
    2=VectorParameter() {parameterName=Color 2, defaultValue=<rgba 0.666 0.0444 0 0>}.RGB
    3=VectorParameter() {parameterName=Color 3, defaultValue=<rgba 0.0333 0.0028 0.0023 0>}.RGB
    A=OneMinus(None=@n4)
   ) {materialFunction=&Lerp_Multiple_Float3}.Lerp 3 Inputs
   B=@n5.RGB
   Alpha=OneMinus(None=Power(Base=@n7.Increased Range, Exp=@n5.A))
  )
  n9 = ScalarParameter() {defaultValue=0.333, parameterName=Erosion Opacity Softness}
  LinearInterpolate(
   A=VectorParameter() {parameterName=Color 5, defaultValue=<rgba 0.0333 0.014 0.0196 1>}.RGB
   B=@n8
   Alpha=MaterialFunctionCall(
    Gradient=@n2
    Erosion Amount=Add(A=@n6, B=@n9) {constB=0.0666}
    Softness=ScalarParameter() {defaultValue=0.333, parameterName=Edge Color Softness}
   ) {materialFunction=&MF_Erosion}.Increased Range
  )
  MaterialFunctionCall(Emissive=LinearInterpolate(A=@n8, B=Multiply(A=@n1.RGB, B=@n3), Alpha=ScalarParameter() {parameterName=UseParticleColor})) {materialFunction=&MF_FX_EmissiveAdjustments}
  Multiply(
   A=MaterialFunctionCall(Gradient=@n2, Erosion Amount=@n6, Softness=@n9) {materialFunction=&MF_Erosion}.Increased Range
   B=ScalarParameter() {defaultValue=0.5, parameterName=Opacity}
  )
  n10 = TextureCoordinate()
  n11 = OneMinus(None=ComponentMask(None=@n10) {r=true})
  n12 = ComponentMask(None=@n10) {g=true}
  n13 = Divide(A=ComponentMask(None=@n10) {r=true}, B=ScalarParameter() {defaultValue=0.166, parameterName=Erosion Base Size})
  n14 = OneMinus(None=SquareRoot(None=Saturate(None=@n13)))
  n15 = Saturate(None=ScalarParameter() {defaultValue=1, parameterName=Erosion Base Pinch})
  n16 = Multiply(A=@n14, B=@n15)
  n17 = Multiply(A=Divide(A=Subtract(A=@n12) {constB=0.5}, B=OneMinus(None=@n16))) {constB=2}
  n18 = DotProduct(A=@n17, B=@n17)
  n19 = OneMinus(None=@n18)
  n20 = Multiply(A=@n14, B=OneMinus(None=@n15))
  Multiply(B=Multiply(A=Multiply(A=@n11, B=@n19), B=OneMinus(None=@n20)))
  n21 = MaterialFunctionCall!(UV, Twist Center, Twist Edges) {materialFunction=/Game/Accessories/FORT_Backpacks/Backpack_F_MED_Celestial_Glow/FX/Materials/MF_PolarUV}
  n22 = MaterialFunctionCall(Float2=@n21.Polar UVs) {materialFunction=&BreakOutFloat2Components}
  n23 = NamedRerouteDeclaration(None=@n22.G) {name=Radial Gradient, nodeColor=<rgba 0.0078 1 0 1>}
  n24 = DynamicParameter() {paramNames [Center Erocion Softness, Center Erocion Offset, Param3, Param4]}
  n25 = Multiply(A=ScalarParameter() {parameterName=Center Erosion Offset}, B=@n24.Center Erocion Offset)
  n26 = ScalarParameter() {defaultValue=0.333, parameterName=Center Erosion Softness}
  n27 = Saturate(None=Divide(A=Subtract(A=@n23, B=@n25), B=Multiply(A=@n26, B=@n24.Center Erocion Softness)))
  n28 = OneMinus(None=@n23)
  n29 = ScalarParameter() {defaultValue=-0.05, parameterName=Edges Erosion Offset}
  n30 = Saturate(None=Divide(A=Subtract(A=@n28, B=@n29), B=ScalarParameter() {defaultValue=0.75, parameterName=Edges Erosion Softness}))
  n31 = Subtract(A=ComponentMask(None=TextureCoordinate()) {r=true}) {constB=0.5}
  n32 = ScalarParameter() {defaultValue=1, parameterName=Erosion Linear Gradient Intensity}
  n33 = MaterialFunctionCall(Base=Multiply(A=Multiply(A=@n27, B=@n30), B=ParticleColor().A), Blend=Add(A=Multiply(A=@n31, B=@n32) {constB=0.5}) {constB=0.5}) {materialFunction=&Blend_Overlay}
  n34 = DepthFade(Opacity=@n33.Result, FadeDistance=ScalarParameter() {defaultValue=22, parameterName=Depth Fade Distance})
  NamedRerouteDeclaration(None=OneMinus(None=QualitySwitch(Default=@n34, Low=@n33.Result, Medium=@n33.Result))) {name=Erosion Amount, nodeColor=<rgba 0.4766 0 1 1>}
  n35 = Multiply(A=@n22.R, B=ScalarParameter() {defaultValue=1, parameterName=UV Angle Tile})
  n36 = Multiply(
   A=Power(Base=@n23, Exp=ScalarParameter() {defaultValue=0.666, parameterName=UV Radial Exponent})
   B=ScalarParameter() {defaultValue=0.5, parameterName=UV Radial Tile}
  )
  n37 = Constant()
  n38 = ScalarParameter() {defaultValue=-0.25, parameterName=UV Panner Speed}
  n39 = TextureSampleParameter2D!(
   UVs=Panner(Coordinate=AppendVector(A=@n35, B=@n36), Speed=AppendVector(A=@n37, B=@n38)) {speedY=-1}
   "DDX(UVs)"=@n21.DDX
   "DDY(UVs)"=@n21.DDX
   Apply View MipBias
  ) {
   mipValueMode=TMVM_Derivative
   parameterName=Param
   samplerType=SAMPLERTYPE_LinearGrayscale
   texture=&/T_Pickaxe_RenegadeRaider_Spark_LavaNoise
  }
  NamedRerouteDeclaration(None=@n39.R) {name=Noise Texture, nodeColor=<rgba 0.3359 0 1 1>}
  NamedRerouteDeclaration(None=@n18) {name=UV Bend, nodeColor=<rgba 0 1 0.0391 1>}
  n40 = Subtract(A=WorldPosition().XYZ, B=ParticlePositionWS())
  PreviousFrameSwitch(Current Frame=Constant3Vector(), Previous Frame=Multiply(A=Normalize(VectorInput=@n40), B=ScalarParameter() {parameterName=Motion Blur}))
 ]
}
